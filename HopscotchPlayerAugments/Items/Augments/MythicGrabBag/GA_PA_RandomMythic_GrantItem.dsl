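; Gameplay ability for the Mythic Grab Bag player augment that grants random mythic item drops from a heist-themed loot table.
; GrantRandomItem picks drops from Loot_Aug_MythicHeistItems, then passes each entry's world item definition and count to GrantItem.
; The activate and end events are disabled stubs, and nothing in this graph calls GrantRandomItem.
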(event Ability|EventActivateAbility
  (|CallParentFunction) ; This node is disabled and will not be called.\nDrag off pins to build functionality.
)

(event Ability|EventOnEndAbility
  (params (bWasCancelled :type Boolean))
  (|CallParentFunction bWasCancelled) ; This node is disabled and will not be called.\nDrag off pins to build functionality.
)

(fn GrantRandomItem
  (bind _outloottodrop (Loot|PickLootDrops "Loot_Aug_MythicHeistItems" 0 -1 "(GameplayTags=)"))
  (for _array_element _outloottodrop
    (CallFunction|GrantItem (Utilities|Casting|CastToFortWorldItemDefinition (Class|FortItemFunctionLibrary|BreakFortItemEntry _array_element)) (.Count (Class|FortItemFunctionLibrary|BreakFortItemEntry _array_element)))))
